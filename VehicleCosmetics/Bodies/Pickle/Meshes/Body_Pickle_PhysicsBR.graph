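# Physics asset for the Pickle vehicle body cosmetic's skeletal mesh, named as the BR variant.
# It holds a single body on the root bone with no collision shapes listed.
# It has no constraints, so it defines no articulated or ragdoll setup.
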PhysicsAsset /VehicleCosmetics/Bodies/Pickle/Meshes/Body_Pickle_PhysicsBR
bodies {root []}
constraints []
